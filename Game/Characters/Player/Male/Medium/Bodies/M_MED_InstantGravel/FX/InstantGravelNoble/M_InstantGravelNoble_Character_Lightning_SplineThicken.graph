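# Translucent unlit Niagara material for lightning effects on the InstantGravel Noble character.
# Spline-thicken math widens geometry perpendicular to view and tangent, varying width along its length, plus a speed-driven offset.
# A zap texture, with Swipe scrolling, random flipping and noise-distorted UVs, drives opacity and cyan or yellow gradients selected by VariantPicker.
# Dynamic parameters steer width, color transition and a noise-edged erosion mask, with depth and optional camera fades.
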Material /Game/Characters/Player/Male/Medium/Bodies/M_MED_InstantGravel/FX/InstantGravelNoble/M_InstantGravelNoble_Character_Lightning_SplineThicken
graph {
 properties {
  bUsedWithNiagaraMeshParticles=true
  bUsedWithNiagaraRibbons=true
  bUsedWithNiagaraSprites=true
  blendMode=BLEND_Translucent
  pixelDepthOffsetMode=PDOM_Legacy
  shadingModel=MSM_Unlit
  twoSided=true
  usedShadingModels=Unlit
 }
 nodes [
  n1 = CrossProduct(
   A=Subtract(A=CameraPositionWS(), B=WorldPosition().XYZ)
   B=MaterialFunctionCall() {materialFunction=&TangentBasis}.Y V
  )
  n2 = ScalarParameter() {defaultValue=6, group=01-Spline, parameterName=TipsWidth, sortPriority=2}
  n3 = ScalarParameter() {defaultValue=15, group=01-Spline, parameterName=MiddleWidth, sortPriority=1}
  n4 = TextureCoordinate()
  n5 = ComponentMask(None=@n4) {g=true}
  n6 = SmoothStep(Value=@n5) {constMax=0.5}
  n7 = OneMinus(None=@n5)
  n8 = Multiply(
   A=LinearInterpolate(
    A=ScalarParameter() {defaultValue=5, group=01-Spline, parameterName=BaseWidth, sortPriority=0}
    B=LinearInterpolate(A=@n2, B=@n3, Alpha=@n6)
    Alpha=SmoothStep(Value=@n7) {constMax=0.5}
   )
   B=DynamicParameter() {defaultValue=<rgba 0 1 1 0>, paramNames [Swipe, Scale U, Width, VariantPicker]}.Width
  )
  n9 = ComponentMask(None=Add(A=Constant() {r=-0.5}, B=@n4)) {r=true}
  n10 = DynamicParameter() {parameterIndex=2, defaultValue=<rgba 0 0 0 1>, paramNames [SpeedX,SpeedY,SpeedZ,Param4]}
  n11 = MaterialFunctionCall(X=@n10.SpeedX, Y=@n10.SpeedX, Z=@n10.SpeedX) {materialFunction=&MakeFloat3}
  n12 = TextureCoordinate()
  n13 = Power(Base=OneMinus(None=ComponentMask(None=@n12) {g=true}))
  Add(
   A=Multiply(A=Normalize(VectorInput=@n1), B=Multiply(A=@n8, B=@n9))
   B=Multiply(A=Multiply(A=@n11.Result) {constB=-1}, B=SmoothStep(Value=@n13))
  )
  n14 = Multiply(A=MaterialFunctionCall(Distance=ScalarParameter() {defaultValue=250, group=03 - Opacity, parameterName=CameraDistance_Fade}) {materialFunction=&Near_Camera_Fade}.Result) {constB=5}
  n15 = ParticleColor()
  n16 = TextureObjectParameter() {
   parameterName=Lightning Parameter
   texture=&/T_InstantGravel_ZapTex
  }
  n17 = TextureSample(UVs=NamedRerouteUsage(), Tex=@n16)
  n18 = Saturate(None=Add(A=SmoothStep(Value=@n17.R) {constMax=0.2}, B=Multiply(A=@n17.G) {constB=0.5}))
  n19 = DepthFade(FadeDistance=ScalarParameter() {group=03 - Opacity, parameterName=DepthFadeDist})
  n20 = Constant() {r=1}
  n21 = Multiply(
   A=Multiply(A=Multiply(A=@n15.A, B=@n18), B=FeatureLevelSwitch(Default=@n19, ES2=@n20, ES3_1=@n20, SM5=@n19, SM6=@n19))
   B=NamedRerouteUsage()
  )
  n22 = DynamicParameter() {parameterIndex=1, defaultValue=<rgba 0 0 0.5 1>, paramNames [ColorTransition, UseCameraFade, MaskPos, MaskSize]}
  LinearInterpolate(A=@n21, B=Multiply(A=Saturate(None=@n14), B=@n21), Alpha=@n22.UseCameraFade)
  n23 = DynamicParameter() {defaultValue=<rgba 0 1 1 0>, paramNames [Swipe, Scale U, Width, VariantPicker]}
  n24 = LinearInterpolate(
   A=VectorParameter() {group=02 - Color, parameterName=CyanDarkColor, defaultValue=<rgba 0.0968 0.0886 0.3698 1>}.RGB
   B=VectorParameter() {group=02 - Color, parameterName=YellowDarkColor, defaultValue=<rgba 0.276 0.1722 0 1>}.RGB
   Alpha=@n23.VariantPicker
  )
  n25 = ComponentMask(None=TextureCoordinate()) {g=true}
  n26 = LinearInterpolate(
   A=LinearInterpolate(
    A=VectorParameter() {group=02 - Color, parameterName=CyanMidDarkColor, defaultValue=<rgba 0.1133 0.0586 0.4167 1>}.RGB
    B=VectorParameter() {group=02 - Color, parameterName=YellowMidDarkColor, defaultValue=<rgba 0.526 0.33 0 1>}.RGB
    Alpha=@n23.VariantPicker
   )
   B=@n24
   Alpha=SmoothStep(Value=@n25) {constMax=0.5, constMin=0.8}
  )
  n27 = LinearInterpolate(
   A=VectorParameter() {group=02 - Color, parameterName=CyanMidColor, defaultValue=<rgba 0.047 0.1212 0.5312 1>}.RGB
   B=VectorParameter() {group=02 - Color, parameterName=YellowMidColor, defaultValue=<rgba 1 0.9109 0.1354 1>}.RGB
   Alpha=@n23.VariantPicker
  )
  n28 = DynamicParameter() {parameterIndex=1, defaultValue=<rgba 0 0 0.5 1>, paramNames [ColorTransition, UseCameraFade, MaskPos, MaskSize]}
  n29 = LinearInterpolate(A=@n27, B=@n24, Alpha=@n28.ColorTransition)
  n30 = LinearInterpolate(
   A=VectorParameter() {group=02 - Color, parameterName=CyanLightColor, defaultValue=<rgba 0 1.5 1.2423 1>}.RGB
   B=VectorParameter() {group=02 - Color, parameterName=YellowLightColor, defaultValue=<rgba 1 0.9912 0.4533 1>}.RGB
   Alpha=@n23.VariantPicker
  )
  n31 = ScalarParameter() {defaultValue=0.5, group=02 - Color, parameterName=ColorTransitionPoint}
  n32 = ScalarParameter() {defaultValue=0.3, group=02 - Color, parameterName=ColorTransitionSoftness}
  n33 = LinearInterpolate(
   A=@n29
   B=LinearInterpolate(A=@n29, B=LinearInterpolate(A=@n30, B=@n27, Alpha=@n28.ColorTransition), Alpha=SmoothStep(Value=@n17.R) {constMax=0.4, constMin=0.1})
   Alpha=SmoothStep(Min=@n31, Max=Add(A=@n31, B=@n32) {constB=0.2}, Value=@n25) {constMax=0.7, constMin=0.5}
  )
  LinearInterpolate(A=@n26, B=@n33, Alpha=SmoothStep(Value=@n25) {constMax=0.4})
  n34 = SmoothStep(Value=@n17.R) {constMax=0.1}
  MaterialFunctionCall(Emissive=Multiply(A=@n15.RGB, B=LinearInterpolate(A=@n26, B=@n33, Alpha=@n34))) {materialFunction=&MF_FX_EmissiveAdjustments}
  n35 = DynamicParameter() {defaultValue=<rgba 0 1 1 0>, paramNames [Swipe, Scale U, Width, VariantPicker]}
  Multiply(A=@n35.Swipe) {constB=0.5}
  n36 = AppendVector(A=@n35.Swipe, B=Constant())
  n37 = Multiply(
   A=TextureCoordinate() {coordinateIndex=1}
   B=ScalarParameter() {defaultValue=0.25, group=00 - UVs, parameterName=Distortion Tiling, sliderMax=5}
  )
  n38 = ConstantBiasScale(None=TextureSample(UVs=Add(A=@n36, B=@n37)) {texture=/Game/TimeOfDay/LightFunction/T_CloudNoise_Artemis}.RGB) {bias=-0.2, scale=2}
  NamedRerouteDeclaration(None=Multiply(A=ComponentMask(None=@n38) {r=true}, B=ScalarParameter() {group=00 - UVs, parameterName=Distortion})) {name=Distortion, nodeColor=<rgba 0 1 0.7891 1>}
  n39 = Round(None=ParticleRandom())
  n40 = MaterialFunctionCall(XY=Multiply(A=TextureCoordinate(), B=LinearInterpolate(Alpha=@n39) {constA=1, constB=-1})) {materialFunction=&Swizzle}
  n41 = AppendVector(A=DynamicParameter() {defaultValue=<rgba 0 1 1 0>, paramNames [Swipe, Scale U, Width, VariantPicker]}.Scale U, B=Constant() {r=1})
  n42 = DynamicParameter() {defaultValue=<rgba 0 1 1 0>, paramNames [Swipe, Scale U, Width, VariantPicker]}
  n43 = Constant()
  n44 = ComponentMask(None=TextureCoordinate()) {g=true}
  n45 = Multiply(A=SmoothStep(Value=@n44) {constMax=0.3, constMin=-0.05}, B=SmoothStep(Value=OneMinus(None=@n44)) {constMax=0.2, constMin=-0.05})
  n46 = MaterialFunctionCall(
   UVs=Add(A=Multiply(A=@n40.YX, B=@n41), B=AppendVector(A=@n42.Swipe, B=@n43))
   Texture Scale=AppendVector(A=Constant() {r=1}, B=Saturate(None=@n45))
  ) {materialFunction=&ScaleUVsByCenter}
  n47 = AppendVector(A=ComponentMask(None=@n46.UVs) {r=true}, B=Saturate(None=ComponentMask(None=@n46.UVs) {g=true}))
  NamedRerouteDeclaration(None=Add(A=@n47, B=NamedRerouteUsage())) {name=Distorted UV, nodeColor=<rgba 0 1 0.0513 1>}
  NamedRerouteDeclaration(None=@n47) {name=UndistortedUVs, nodeColor=<rgba 1 0.8672 0 1>}
  n48 = DynamicParameter() {parameterIndex=1, defaultValue=<rgba 0 0 0.5 1>, paramNames [ColorTransition, UseCameraFade, MaskPos, MaskSize]}
  n49 = OneMinus(None=@n48.MaskSize)
  n50 = ScalarParameter() {defaultValue=0.035, parameterName=MaskSoftness}
  n51 = OneMinus(None=TextureCoordinate())
  n52 = Multiply(A=TextureCoordinate() {coordinateIndex=1}, B=Constant2Vector() {g=0.1, r=1})
  n53 = Multiply(A=ComponentMask(None=TextureSample(UVs=@n52) {samplerType=SAMPLERTYPE_Normal, texture=/Game/Global/Textures/Default/Noise/T_Bubble_B_N}.RGB) {g=true, r=true}) {constB=0.1}
  n54 = Add(A=@n51, B=@n53)
  n55 = Add(A=Saturate(None=ComponentMask(None=@n54) {g=true}))
  n56 = DynamicParameter() {parameterIndex=1, defaultValue=<rgba 0 0 0.5 1>, paramNames [ColorTransition, UseCameraFade, MaskPos, MaskSize]}
  n57 = LinearInterpolate(A=Constant() {r=0.4}, B=@n56.MaskPos, Alpha=Ceil(None=ParticleRelativeTime() {desc=">0"})) {constA=0.5}
  n58 = Multiply(A=Saturate(None=Multiply(A=@n57) {constB=2})) {constB=2}
  n59 = OneMinus(None=Saturate(None=Subtract(A=@n55, B=@n58)))
  n60 = Saturate(None=ComponentMask(None=Add(A=@n51, B=@n53)) {g=true})
  n61 = Saturate(None=Subtract(A=Multiply(A=@n57) {constB=2}) {constB=0.5})
  n62 = Saturate(None=Subtract(A=Add(A=@n60), B=Multiply(A=@n61) {constB=2}))
  NamedRerouteDeclaration(None=SmoothStep(Min=@n49, Max=Add(A=@n49, B=@n50), Value=Multiply(A=@n59, B=@n62)) {constMax=0.9, constMin=0.9}) {name=Erosion, nodeColor=<rgba 0.2793 0 1 1>}
  ScalarParameter() {defaultValue=0.4, parameterName=MaskSize}
  TextureSample(UVs=NamedRerouteUsage(), Tex=@n16)
 ]
}
