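# Niagara one-shot burst for the InstantGravel Noble player body, named for the start of a sprint, anchored to sockets on the character's skeletal mesh.
# BurstFlare adds a shrinking sprite flare with a pulsed light, while ChestRadialBurst throws short-lived, drag-slowed streaks radially from a socket and inherits source movement.
# Three lightning emitters are disabled: back zaps, back speed lines, and chest zap meshes that feed motion-based bending into material parameters.
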NiagaraSystem /Game/Characters/Player/Male/Medium/Bodies/M_MED_InstantGravel/FX/InstantGravelNoble/NS_InstantGravelNoble_SprintBurst
dependencies {
 usedDataInterfaces [
  &NiagaraDataInterfaceArrayFloat3
  &NiagaraDataInterfaceVectorCurve
  &NiagaraDataInterfaceCurve
  &NiagaraDataInterfaceColorCurve
  &NiagaraDataInterfaceVector2DCurve
 ]
 usedDynamicInputs [&MakeFloatFromVector2d]
 usedModules [
  &self:SystemSpawnScript.SystemScriptSource.SystemScriptGraph.NiagaraNodeAssignment_0.SetVariables_#1
  &SystemState
  &self:SystemSpawnScript.SystemScriptSource.SystemScriptGraph.NiagaraNodeAssignment_8.SetVariables_#2
  &self:SystemSpawnScript.SystemScriptSource.SystemScriptGraph.NiagaraNodeAssignment_3.SetVariables_#3
 ]
}
emitters {
 BackZaps {
  bEnabled=false
  emitterSpawnScript [
   /SetVariables(Emitter.EmitRefract=1.5) {isSetParameters=true, module=SetVariables_#4}
   /SetVariables(Emitter.EmitColorLightning:LinearColor=$User.EnergyColor) {isSetParameters=true, module=SetVariables_#5}
  ]
  emitterUpdateScript [
   &EmitterState()
   &SpawnBurst_Instantaneous(Spawn Count=15, Spawn Time=0.05, Use Spawn Probability=false, UseLoopCountLimit=false)
  ]
  particleSpawnScript [
   &InitializeParticle(
    Lifetime Max=0.6
    Lifetime Min=0.4
    Lifetime Mode=::NewEnumerator1
    Sprite Size=&RandomRangeVector2D(Minimum=<10 115>, Maximum=<15 165>)
    Sprite Size Mode=::NewEnumerator4
    UsePositionOffset=false
    Write Material Random=true
   )
   /SetVariables(Particles.Age:NiagaraFloat=&RandomRangeFloat(Minimum=0, Maximum=-0.4)) {isSetParameters=true}
   /SetVariables(Particles.ZapIndex:NiagaraInt32=&RandomRangeInt(Minimum=0, Maximum=7)) {isSetParameters=true}
   &SkeletalMeshLocation(
    Bone / Socket Index=&RandomRangeInt(Minimum=8, Maximum=13)
    MaskSpawnGroup=false
    Mesh Sampling Type=::NewEnumerator5
    Sampled Position Offset=&SelectVectorFromArray(Direct Array Index=$Particles.ZapIndex, Direct Array Mode=::NewEnumerator1, Vector Selection Array=$System.BackBoneOffsetVectorArray)
    Skeletal Mesh=$User.SkeletalMesh
    Socket Sampling Mode=::NewEnumerator5
   )
   &AddVelocity(Velocity=&Multiply_VectorByFloat(
     Vector=&MakeCustomVectorFromBool(Bool=$User.Debug, True Vector=<0 1 1>, False Vector=<0 1 0>)
     Float=&RandomRangeFloat(Minimum=-50, Maximum=-150)
    ))
   &ApplyInitialForces()
  ]
  particleUpdateScript [
   &ParticleState()
   &ScaleSpriteSize(
    Non-Uniform Curve Sprite Scale=DataInterface() {propertyValues {
      XCurve {keys [{tangentMode=RCTM_User, time=0, value=1}, {tangentMode=RCTM_User, time=1, value=0.1}]}
      YCurve {keys [{tangentMode=RCTM_User, time=0, value=0}, {tangentMode=RCTM_User, time=1, value=2}]}
     }}
    Scale Sprite Size Mode=::NewEnumerator3
   ) {module=ScaleSpriteSize001}
   &SkeletalMeshLocation(
    MaskSpawnGroup=false
    Mesh Sampling Type=::NewEnumerator5
    Sampled Position Offset=&Add_Vector(
     A=&SelectVectorFromArray(Direct Array Index=$Particles.ZapIndex, Direct Array Mode=::NewEnumerator1, Vector Selection Array=$System.BackBoneOffsetVectorArray)
     B=&RandomRangeVector(Minimum=<-15 0 -15>, Maximum=<15 0 15>)
    )
    Skeletal Mesh=$User.SkeletalMesh
    Socket Sampling Mode=::NewEnumerator5
   ) {module=SkeletalMeshLocation001}
   &InheritSourceMovement(
    Applied Position Scale=&VectorFromFloat(Value=&FloatFromCurve(FloatCurve=DataInterface() {propertyValues {Curve {keys [<at 0 1>,<at 1 1>]}}}))
    ApplyForce=false
    ApplyPosition=true
    ApplyPositionFalloff=false
   )
   &SolveForcesAndVelocity(Clamp Velocity=false, Limit Acceleration=false)
   &CameraOffset(Camera Offset Amount=8)
   &DynamicMaterialParameters(
    Index 0 Param 1=&RandomRangeFloat(Minimum=0, Maximum=1)
    Index 0 Param 2=&RandomRangeFloat(Minimum=0.4, Maximum=0.5)
    Index 0 Param 4=0.01
    Index 1 Param 1=0.01
    Index 1 Param 3=&FloatFromCurve(FloatCurve=DataInterface() {propertyValues {Curve {keys [<at 0 1>,<at 1 0>]}}})
    Index 1 Param 4=0.5
    Module.Param0WriteEnabled=true
    Module.Param0WriteEnabled1=true
    Module.Param1WriteEnabled=true
    Module.Param1WriteEnabled1=true
    Module.Param2WriteEnabled=true
    Module.Param2WriteEnabled1=true
    Module.Param3WriteEnabled=true
    Module.Param3WriteEnabled1=true
    Write Parameter Index 1=true
   )
   &Color(Color=&MultiplyLinearColor(
     Linear Color=&ColorFromCurve(
      ColorCurve=DataInterface() {propertyValues {
        AlphaCurve {keys [<at 0 1 RCIM_Cubic>]}
        BlueCurve {keys [<at 0.3761 3 RCIM_Cubic>,<at 0.5389 3>]}
        GreenCurve {keys [<at 0.3761 3 RCIM_Cubic>,<at 0.5389 2.7584>]}
        RedCurve {keys [<at 0.3761 0 RCIM_Cubic>,<at 0.5389 2.0599>]}
       }}
      CurveIndex=&RandomRangeFloat(Minimum=0, Maximum=1)
     )
     Scale Factor=&MakeFloat4FromVectorAndFloat(
      "Vector (RGB)"=&VectorFromFloat(Value=&FloatFromCurve(FloatCurve=DataInterface() {propertyValues {Curve {keys [<at 0 1>,<at 1 0.5>]}}}))
      "Float (Alpha)"=1
     )
    ))
   &SpriteFacingAndAlignment(
    Sprite Alignment=/NiagaraFluids/DynamicInput/MaxVectorByLength(
     A:Vector3f=&NormalizeVector(
      Vector To Normalize=&Multiply_VectorByFloat(Vector=$Engine.Owner.Velocity, Float=-1)
      Safe Normalize=true
      Safe Fallback Threshold=0
      Safe Fallback Vector=<0 0 0.001>
     )
     B:Vector3f=<0 -0.01 0>
    )
    WriteAlignment=true
    WriteFacing=true
   )
   &ScaleSpriteSize(
    Scale Factor=&MakeVector2D(
     X=&Multiply_Float(
      A=&FloatFromCurve(
       FloatCurve=DataInterface() {propertyValues {Curve {keys [
           {arriveTangent=-3.3422, interpMode=RCIM_Cubic, leaveTangent=-3.3422, tangentMode=RCTM_User, time=0, value=1}
           {arriveTangent=-0.6568, interpMode=RCIM_Cubic, leaveTangent=-0.6568, tangentMode=RCTM_User, time=1, value=0}
          ]}}}
       Scale Curve=30
      )
      B=&FloatFromCurve(
       FloatCurve=DataInterface() {propertyValues {Curve {keys [{tangentMode=RCTM_User, time=0, value=0.5}, {tangentMode=RCTM_User, time=1, value=1}]}}}
       CurveIndex=&VectorLength(Vector=$Engine.Owner.Velocity)
      )
     )
     Y=&Lerp_Float(
      A=1
      B=6
      Alpha=&FloatFromCurve(
       FloatCurve=DataInterface() {propertyValues {Curve {keys [
           {interpMode=RCIM_Cubic, tangentMode=RCTM_User, time=0, value=0}
           {arriveTangent=2.4302, interpMode=RCIM_Cubic, leaveTangent=2.4302, tangentMode=RCTM_User, time=1, value=1}
          ]}}}
       CurveIndex=&VectorLength(Vector=$Engine.Owner.Velocity)
      )
     )
    )
    Scale Sprite Size Mode=::NewEnumerator2
   ) {enabled=false}
  ]
  rendererClasses [&NiagaraSpriteRendererProperties]
  renderers [{rendererClass=&NiagaraSpriteRendererProperties}]
 }
 BurstFlare {
  emitterUpdateScript [
   &EmitterState()
   &SpawnBurst_Instantaneous(Use Spawn Probability=false, UseLoopCountLimit=false) {module=SpawnBurst_Instantaneous001}
  ]
  particleSpawnScript [
   &InitializeParticle(
    Color=<rgba 1 1 20 0.1>
    Lifetime=0.4
    Sprite Size Mode=::NewEnumerator3
    Uniform Sprite Size=200
    UsePositionOffset=false
    Write Material Random=true
   )
   &SkeletalMeshLocation(
    Bone / Socket Index=&RandomRangeInt(Minimum=8, Maximum=13)
    MaskSpawnGroup=false
    Mesh Sampling Type=::NewEnumerator5
    Skeletal Mesh=$User.SkeletalMesh
    Socket Sampling Mode=::NewEnumerator5
   )
   &SystemLocation(MaskSpawnGroup=false, Offset=<0 0 10>)
  ]
  particleUpdateScript [
   &ParticleState()
   &SkeletalMeshLocation(
    Bone / Socket Index=1
    MaskSpawnGroup=false
    Mesh Sampling Type=::NewEnumerator5
    Skeletal Mesh=$User.SkeletalMesh
    Socket Sampling Mode=::NewEnumerator5
   ) {module=SkeletalMeshLocation001}
   &SolveForcesAndVelocity(Clamp Velocity=false, Limit Acceleration=false)
   &ScaleSpriteSize(Uniform Curve Sprite Scale=DataInterface() {propertyValues {Curve {keys [
        {arriveTangent=1.8001, interpMode=RCIM_Cubic, leaveTangent=3.5392, tangentMode=RCTM_Break, time=0, value=0.5}
        {interpMode=RCIM_Cubic, tangentMode=RCTM_Break, time=1, value=0}
       ]}}})
   &CameraOffset(Camera Offset Amount=20)
   &ScaleColor(
    Scale Alpha=&FloatFromCurve(FloatCurve=DataInterface() {propertyValues {Curve {keys [<at 0 0.2>,<at 1 0>]}}})
    ScaleA=true
    ScaleRGB=true
   )
   &FN_Light_Attributes(
    Color=$Particles.Color
    Create Color Attribute=true
    Create Exponent Attribute=true
    Create Position Attribute=true
    Create Radius Attribute=true
    Position=&AddVectorToPosition(
     Position=$Particles.Position
     Vector=&RotateVectorByQuaternion(Vector To Rotate=<0 20 0>, Quaternion=$Particles.SkeletalMeshLocation001.SampledRotation)
    )
    Radius=&FloatFromCurve(
     FloatCurve=DataInterface() {propertyValues {Curve {keys [
         {interpMode=RCIM_Cubic, tangentMode=RCTM_User, time=0, value=0}
         {interpMode=RCIM_Cubic, tangentMode=RCTM_User, time=0.2, value=1}
         {interpMode=RCIM_Cubic, tangentMode=RCTM_User, time=1, value=0}
        ]}}}
     Scale Curve=50
    )
   )
  ]
  rendererClasses [&NiagaraSpriteRendererProperties, &NiagaraLightRendererProperties]
  renderers [
   {rendererClass=&NiagaraSpriteRendererProperties}
   {rendererClass=&NiagaraLightRendererProperties, rendererIndex=1}
  ]
 }
 ChestRadialBurst {
  emitterSpawnScript [
   /SetVariables(Emitter.EmitRefract=1.5) {isSetParameters=true, module=SetVariables_#4}
   /SetVariables() {isSetParameters=true, module=SetVariables_#5}
  ]
  emitterUpdateScript [
   &EmitterState()
   &SpawnBurst_Instantaneous(Spawn Count=15, Use Spawn Probability=false, UseLoopCountLimit=false)
  ]
  particleSpawnScript [
   &InitializeParticle(
    Lifetime Max=0.3
    Lifetime Min=0.2
    Lifetime Mode=::NewEnumerator1
    Sprite Size Max=<2 2>
    Sprite Size Min=<1 1>
    Sprite Size Mode=::NewEnumerator2
    UsePositionOffset=false
    Write Material Random=true
   )
   /SetVariables(Particles.Age:NiagaraFloat=&RandomRangeFloat(Minimum=0, Maximum=-0.4)) {isSetParameters=true}
   /SetVariables(Particles.ZapIndex:NiagaraInt32=&RandomRangeInt(Minimum=0, Maximum=7)) {isSetParameters=true, module=SetVariables_#6}
   &SkeletalMeshLocation(
    Bone / Socket Index=1
    MaskSpawnGroup=false
    Mesh Sampling Type=::NewEnumerator5
    Skeletal Mesh=$User.SkeletalMesh
    Socket Sampling Mode=::NewEnumerator5
   )
   &ApplyInitialForces()
   &ShapeLocation(
    Additional Quaternion Rotation=false
    MaskSpawnGroup=false
    Rotation Mode=::NewEnumerator3
    Rotation Quaternion=$Particles.SkeletalMeshLocation.SampledRotation
    Shape Origin=$Particles.Position
    Sphere Distribution=::NewEnumerator1
    Sphere Radius=5
    U Position=&RandomRangeFloat(Minimum=0, Maximum=1)
    UseOwnerScale=true
   )
   &AddVelocity(
    Velocity=&NormalizeVector(Vector To Normalize=$Particles.ShapeLocation.ShapeVector, Safe Normalize=true, Safe Fallback Threshold=0, Safe Fallback Vector=<0 0 0.001>)
    Velocity Speed Scale=&RandomRangeFloat(Minimum=50, Maximum=150)
   )
   &AddVelocity(Velocity=<0 1 0>, Velocity Speed Scale=&RandomRangeFloat(Minimum=50, Maximum=150)) {module=AddVelocity001}
  ]
  particleUpdateScript [
   &ParticleState()
   &SkeletalMeshLocation(
    MaskSpawnGroup=false
    Mesh Sampling Type=::NewEnumerator5
    Position Sampling=::NewEnumerator2
    Sampled Position Offset=&Add_Vector(
     A=&SelectVectorFromArray(Direct Array Index=$Particles.ZapIndex, Direct Array Mode=::NewEnumerator1, Vector Selection Array=$System.BackBoneOffsetVectorArray)
     B=&RandomRangeVector(Minimum=<-5 0 -5>, Maximum=<5 0 5>)
    )
    Skeletal Mesh=$User.SkeletalMesh
    Socket Sampling Mode=::NewEnumerator5
   ) {module=SkeletalMeshLocation001}
   &Drag(Drag=&RandomRangeFloat(Minimum=8, Maximum=12), Use Linear Drag=true, Use Rotational Drag=true)
   &InheritSourceMovement(
    Applied Position Scale=&VectorFromFloat(Value=&FloatFromCurve(FloatCurve=DataInterface() {propertyValues {Curve {keys [<at 0 1>,<at 1 0.8>]}}}))
    ApplyForce=false
    ApplyPosition=true
    ApplyPositionFalloff=false
   )
   &SolveForcesAndVelocity(Clamp Velocity=false, Limit Acceleration=false)
   &Color(Color=&MultiplyLinearColor(
     Linear Color=&ColorFromCurve(
      ColorCurve=DataInterface() {propertyValues {
        AlphaCurve {keys [<at 0 1 RCIM_Cubic>]}
        BlueCurve {keys [<at 0.4286 1>,<at 0.5381 2>]}
        GreenCurve {keys [<at 0.4286 1>,<at 0.5381 2>]}
        RedCurve {keys [<at 0.4286 1>,<at 0.5381 2>]}
       }}
      CurveIndex=&RandomRangeFloat(Minimum=0, Maximum=1)
     )
     Scale Factor=&MakeFloat4FromVectorAndFloat(
      "Vector (RGB)"=&VectorFromFloat(Value=&FloatFromCurve(FloatCurve=DataInterface() {propertyValues {Curve {keys [<at 0 1>,<at 1 0.5>]}}}))
      "Float (Alpha)"=1
     )
    ))
   &CameraOffset()
   &DynamicMaterialParameters(
    Index 0 Param 1=&FloatFromCurve(
     FloatCurve=DataInterface() {propertyValues {Curve {keys [<at 0 1>,<at 1 0>]}}}
     Scale Curve=&RandomRangeFloat(Minimum=0.4, Maximum=0.6)
    )
    Index 0 Param 2=&RandomRangeFloat(Minimum=0.2, Maximum=0.4)
    Index 0 Param 4=0.01
    Index 1 Param 1=0.02
    Index 1 Param 2=1
    Index 1 Param 3=&FloatFromCurve(FloatCurve=DataInterface() {propertyValues {Curve {keys [{arriveTangent=1.9378, time=0, value=0}, {arriveTangent=2.3849, leaveTangent=2.3849, time=1, value=1}]}}})
    Index 1 Param 4=0.8
    Module.Param0WriteEnabled=true
    Module.Param0WriteEnabled1=true
    Module.Param1WriteEnabled=true
    Module.Param1WriteEnabled1=true
    Module.Param2WriteEnabled=true
    Module.Param2WriteEnabled1=true
    Module.Param3WriteEnabled=true
    Module.Param3WriteEnabled1=true
    Write Parameter Index 1=true
   )
   &ScaleSpriteSize(
    Scale Factor=&Multiply_Vector2D(
     A=<10 30>
     B=&Vector2DFromCurve(Vector2Curve=DataInterface() {propertyValues {XCurve {keys [<at 0 1>,<at 1 0>]}, YCurve {keys [<at 0 0>,<at 0.1 1>]}}}, ScaleCurve=<1 1>)
    )
    Scale Sprite Size Mode=::NewEnumerator2
   )
  ]
  rendererClasses [&NiagaraSpriteRendererProperties]
  renderers [{rendererClass=&NiagaraSpriteRendererProperties}]
 }
 NS_ChestZaps {
  bEnabled=false
  emitterUpdateScript [
   &EmitterState()
   &SpawnBurst_Instantaneous(Spawn Count=12, Use Spawn Probability=false, UseLoopCountLimit=false)
  ]
  particleSpawnScript [
   /SetVariables(Particles.RandomRot:NiagaraFloat=&RandomRangeFloat(Minimum=0, Maximum=360)) {isSetParameters=true}
   &PartitionParticles(Number Of Partitions=8)
   /SetVariables(Particles.ZapIndex:NiagaraInt32=$Particles.PartitionParticles.Partition) {isSetParameters=true}
   /SetVariables() {isSetParameters=true}
   &InitializeParticle(
    Color Mode=::NewEnumerator0
    Lifetime=&RandomRangeFloat(Minimum=0.2, Maximum=0.4)
    Mesh Scale=<0.7 0.5 1>
    Mesh Scale Mode=::NewEnumerator4
    Sprite Size Mode=::NewEnumerator3
    UsePositionOffset=false
    Write Material Random=true
   )
   /SetVariables(Particles.Age:NiagaraFloat=&RandomRangeFloat(Minimum=0, Maximum=-0.4)) {isSetParameters=true}
   &InitialMeshOrientation(
    Mesh Orientation Mode=::NewEnumerator7
    Rotation=&Add_Vector(
     A=&SelectVectorFromArray(
      Direct Array Index=$Particles.ZapIndex
      Vector Selection Array=DataInterface() {propertyValues {FloatData [<0 -0.05 0>, <0 -0.1 0>, <0 -0.15 0>, <0 -0.2 0>, <0 0.05 0>, <0 0.1 0>, <0 0.15 0>, <0 0.2 0>]}}
     )
     B=<0 0 0>
    )
    Use Rotation Vector=true
   )
   &SkeletalMeshLocation(
    Bone / Socket Index=1
    MaskSpawnGroup=false
    Mesh Sampling Type=::NewEnumerator5
    Skeletal Mesh=$User.SkeletalMesh
    Socket Sampling Mode=::NewEnumerator5
   ) {module=SkeletalMeshLocation001}
  ]
  particleUpdateScript [
   &ParticleState()
   &ScaleMeshSize(Scale Factor=&Multiply_VectorByFloat(
     Vector=&VectorFromCurve(VectorCurve=DataInterface() {propertyValues {XCurve {keys [<at 0.1 1>,<at 1 1.2>]}, YCurve {keys [<at 0.1 1>,<at 1 1>]}, ZCurve {keys [<at 0 0>,<at 1 1.2>]}}})
     Float=1
    ))
   /SetVariables(Particles.PreviousPos:NiagaraPosition=$Particles.Position) {isSetParameters=true}
   &SkeletalMeshLocation(
    Bone / Socket Index=1
    MaskSpawnGroup=false
    Mesh Sampling Type=::NewEnumerator5
    Position Sampling=::NewEnumerator2
    Skeletal Mesh=$User.SkeletalMesh
    Socket Sampling Mode=::NewEnumerator5
   )
   /SetVariables(Particles.Position=&AddVectorToPosition(
     Position=$Particles.SkeletalMeshLocation.SampledPosition
     Vector=&RotateVectorByQuaternion(
      Vector To Rotate=&SelectVectorFromArray(
       Direct Array Index=$Particles.ZapIndex
       Vector Selection Array=DataInterface() {propertyValues {FloatData [<-10 0 3>, <-13 0 6>, <-16 0 9>, <-19 0 12>, <-10 0 -3>, <-13 0 -6>, <-16 0 -9>, <-19 0 -12>]}}
      )
      Quaternion=$Particles.SkeletalMeshLocation.SampledRotation
     )
    )) {isSetParameters=true}
   &self:SetZapBending(CurrentPos:NiagaraPosition=$Particles.Position, MaxBending=50, RestSpeed=3, VelDivider=2) {module=SetZapBending}
   &DynamicMaterialParameters(
    Index 0 Param 1=&Add_Float(
     A=&RandomRangeFloat(Minimum=0, Maximum=1)
     B=&FloatFromCurve(
      FloatCurve=DataInterface() {propertyValues {Curve {keys [<at 0 0>,<at 1 1>]}}}
      Scale Curve=&Lerp_Float(A=0.6, B=0.3, Alpha=$System.SpeedFactor)
     )
    )
    Index 0 Param 2=1
    Index 0 Param 3=&FloatFromCurve(FloatCurve=DataInterface() {propertyValues {Curve {keys [<at 0 0.5>,<at 1 1>]}}})
    Index 0 Param 4=0.01
    Index 1 Param 1=0.02
    Index 1 Param 2=1
    Index 1 Param 3=&FloatFromCurve(FloatCurve=DataInterface() {propertyValues {Curve {keys [{arriveTangent=0.2596, time=0, value=1}, {arriveTangent=1.9883, leaveTangent=1.9883, time=1, value=0}]}}})
    Index 1 Param 4=0.5
    Index 2 Param 1=&MakeFloatFromVector(Vector=$Particles.ZapBending, Channel=::NewEnumerator0)
    Index 2 Param 2=&MakeFloatFromVector(Vector=$Particles.ZapBending, Channel=::NewEnumerator1)
    Index 2 Param 3=&MakeFloatFromVector(Vector=$Particles.ZapBending, Channel=::NewEnumerator2)
    Module.Index0Param0Write=true
    Module.Index0Param1Write=true
    Module.Index0Param2Write=true
    Module.Index0Param3Write=true
    Module.Index1Param0Write=true
    Module.Index1Param1Write=true
    Module.Index1Param2Write=true
    Module.Index1Param3Write=true
    Module.Index2Param0Write=true
    Module.Index2Param1Write=true
    Module.Index2Param2Write=true
    Module.Index2Param3Write=true
    Write Parameter Index 1=true
    Write Parameter Index 2=true
   )
   &Color(Color=&ColorFromCurve(
     ColorCurve=DataInterface() {propertyValues {
       AlphaCurve {keys [<at 0 1 RCIM_Cubic>]}
       BlueCurve {keys [<at 0.4201 1>,<at 0.5015 2>]}
       GreenCurve {keys [<at 0.4201 1>,<at 0.5015 2>]}
       RedCurve {keys [<at 0.4201 1>,<at 0.5015 2>]}
      }}
     CurveIndex=&RandomRangeFloat(Minimum=0, Maximum=1)
    ))
  ]
  rendererClasses [&NiagaraMeshRendererProperties]
  renderers [{rendererClass=&NiagaraMeshRendererProperties, rendererIndex=0}]
 }
 SmallZaps_Back_SpeedLines {
  bEnabled=false
  emitterSpawnScript [
   /SetVariables(Emitter.EmitRefract=1.5) {isSetParameters=true, module=SetVariables_#4}
   /SetVariables() {isSetParameters=true, module=SetVariables_#5}
  ]
  emitterUpdateScript [
   &EmitterState()
   &SpawnBurst_Instantaneous(Spawn Count=20, Spawn Time=0.1, Use Spawn Probability=false, UseLoopCountLimit=false)
  ]
  particleSpawnScript [
   &InitializeParticle(
    Lifetime Max=0.4
    Lifetime Min=0.2
    Lifetime Mode=::NewEnumerator1
    Sprite Size=&MakeVector2D(
     X=&RandomRangeFloat(Minimum=0.5, Maximum=1.5)
     Y=&RandomRangeFloat(Minimum=0.5, Maximum=1.5)
    )
    Sprite Size Mode=::NewEnumerator4
    UsePositionOffset=false
    Write Material Random=true
   )
   /SetVariables(Particles.ZapIndex:NiagaraInt32=&RandomRangeInt(Minimum=0, Maximum=7)) {isSetParameters=true, module=SetVariables_#6}
   &SkeletalMeshLocation(
    MaskSpawnGroup=false
    Mesh Sampling Type=::NewEnumerator5
    Sampled Position Offset=&Add_Vector(
     A=&SelectVectorFromArray(Direct Array Index=$Particles.ZapIndex, Direct Array Mode=::NewEnumerator1, Vector Selection Array=$System.BackBoneOffsetVectorArray)
     B=&RandomRangeVector(Minimum=<-5 20 -5>, Maximum=<5 30 5>)
    )
    Skeletal Mesh=$User.SkeletalMesh
    Socket Sampling Mode=::NewEnumerator5
   )
   &AddVelocity(Velocity=&RandomRangeVector(Minimum=<0 -50 0>, Maximum=<1 -150 1>))
   &ApplyInitialForces()
  ]
  particleUpdateScript [
   &ParticleState()
   &SkeletalMeshLocation(
    MaskSpawnGroup=false
    Mesh Sampling Type=::NewEnumerator5
    Position Sampling=::NewEnumerator2
    Sampled Position Offset=&Add_Vector(
     A=&SelectVectorFromArray(Direct Array Index=$Particles.ZapIndex, Direct Array Mode=::NewEnumerator1, Vector Selection Array=$System.BackBoneOffsetVectorArray)
     B=&RandomRangeVector(Minimum=<-5 0 -5>, Maximum=<5 0 5>)
    )
    Skeletal Mesh=$User.SkeletalMesh
    Socket Sampling Mode=::NewEnumerator5
   ) {module=SkeletalMeshLocation001}
   &InheritSourceMovement(
    Applied Position Scale=&VectorFromFloat(Value=&FloatFromCurve(FloatCurve=DataInterface() {propertyValues {Curve {keys [<at 0 1>,<at 1 0.8>]}}}))
    ApplyForce=false
    ApplyPosition=true
    ApplyPositionFalloff=false
   )
   &SolveForcesAndVelocity(Clamp Velocity=false, Limit Acceleration=false)
   &Color(Color=&ColorFromCurve(ColorCurve=DataInterface() {propertyValues {
       AlphaCurve {keys [<at 0 1 RCIM_Cubic>]}
       BlueCurve {keys [<at 0 3 RCIM_Cubic>]}
       GreenCurve {keys [<at 0 3 RCIM_Cubic>]}
       RedCurve {keys [<at 0 0 RCIM_Cubic>]}
      }}))
   &CameraOffset(Camera Offset Amount=8)
   &DynamicMaterialParameters(
    Index 0 Param 1=&FloatFromCurve(
     FloatCurve=DataInterface() {propertyValues {Curve {keys [<at 0 1>,<at 1 0>]}}}
     Scale Curve=&RandomRangeFloat(Minimum=0.4, Maximum=0.6)
    )
    Index 0 Param 2=&RandomRangeFloat(Minimum=0.2, Maximum=0.4)
    Index 0 Param 4=0.01
    Index 1 Param 1=0.02
    Index 1 Param 2=1
    Index 1 Param 3=&FloatFromCurve(FloatCurve=DataInterface() {propertyValues {Curve {keys [{arriveTangent=1.9378, time=0, value=1}, {arriveTangent=2.3849, leaveTangent=2.3849, time=1, value=0}]}}})
    Index 1 Param 4=0.8
    Module.Param0WriteEnabled=true
    Module.Param0WriteEnabled1=true
    Module.Param1WriteEnabled=true
    Module.Param1WriteEnabled1=true
    Module.Param2WriteEnabled=true
    Module.Param2WriteEnabled1=true
    Module.Param3WriteEnabled=true
    Module.Param3WriteEnabled1=true
    Write Parameter Index 1=true
   )
   &ScaleSpriteSize(
    Scale Factor=&Multiply_Vector2D(
     A=<20 40>
     B=&Vector2DFromCurve(Vector2Curve=DataInterface() {propertyValues {XCurve {keys [<at 0 1>,<at 1 0>]}, YCurve {keys [<at 0 0.2>,<at 1 2>]}}}, ScaleCurve=<1 1>)
    )
    Scale Sprite Size Mode=::NewEnumerator2
   )
  ]
  rendererClasses [&NiagaraSpriteRendererProperties]
  renderers [{rendererClass=&NiagaraSpriteRendererProperties}]
 }
}
summary {
 systemName=NS_InstantGravelNoble_SprintBurst
 emitters [
  {bEnabled=false, emitterName=NS_ChestZaps, rendererClasses [&NiagaraMeshRendererProperties]}
  {bEnabled=false, emitterName=SmallZaps_Back_SpeedLines, rendererClasses [&NiagaraSpriteRendererProperties]}
  {emitterName=ChestRadialBurst, rendererClasses [&NiagaraSpriteRendererProperties]}
  {bEnabled=false, emitterName=BackZaps, rendererClasses [&NiagaraSpriteRendererProperties]}
  {emitterName=BurstFlare, rendererClasses [&NiagaraSpriteRendererProperties, &NiagaraLightRendererProperties]}
 ]
 userVariables [
  User.BariumPaintColor {type=LinearColor, description="", defaultValue {struct=LinearColor, value=<rgba 1 1 1 1>}}
  User.bIsFrontEnd {type=NiagaraBool}
  User.bIsFrontEndPreview {type=NiagaraBool}
  User.CinematicVelocity {type=Vector3f, description="", defaultValue {struct=Vector3f, value=<0 200 200>}}
  User.Debug {type=NiagaraBool}
  User.Distance {type=NiagaraFloat, defaultValue=0, description=""}
  User.EnergyColor {type=LinearColor, description="", defaultValue {struct=LinearColor, value=<rgba 1 1 1 1>}}
  User.LightningAcceleration {type=NiagaraFloat, defaultValue=0, description=""}
  User.RibbonLengthFactor {type=NiagaraFloat, defaultValue=1, description=""}
  User.SkeletalMesh {
   type=NiagaraDataInterfaceSkeletalMesh
   defaultValue {
    struct=NiagaraExt_VariableValue_DataInterface
    value {dataInterface=&self:NiagaraDataInterfaceSkeletalMesh_0}
   }
  }
  User.SpawnHelmetFX {type=NiagaraFloat, defaultValue=0, description=""}
  User.SpawnJacketFX {type=NiagaraFloat, defaultValue=0, description=""}
  User.UseCinematicVelocity {type=NiagaraBool}
 ]
}
